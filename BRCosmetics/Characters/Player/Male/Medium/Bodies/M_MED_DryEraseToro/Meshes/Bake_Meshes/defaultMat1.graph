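# Baked default material for the male medium DryEraseToro player body mesh, flagged for Nanite and apparently created by Interchange import.
# It uses Default Lit shading with a diffuse texture parameter.
# A linear interpolate blends from black toward that texture's RGB, driven by a scalar weight parameter.
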Material /BRCosmetics/Characters/Player/Male/Medium/Bodies/M_MED_DryEraseToro/Meshes/Bake_Meshes/defaultMat1
graph {
 properties {assetImportData="<self>:InterchangeAssetImportData_0", bUsedWithNanite=true, usedShadingModels=Default Lit}
 nodes [LinearInterpolate(
   A=Constant3Vector() {constant=<rgba 0 0 0 1>}
   B=TextureSampleParameter2D() {parameterName=DiffuseColorMap}.RGB
   Alpha=ScalarParameter() {defaultValue=0.8, parameterName=DiffuseColorMapWeight}
  )]
}
